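# Post-process AnimBlueprint for the FeralTrash player head, adding secondary motion and pose corrections.
# Ear master bones are repositioned before caching, then two head counter-rotations, scaled from the head-to-neck_02 delta (zero when emoting), blend on emote state.
# AnimDynamics then simulates ears and cheeks, and outer lower lip bones rotate during the Emote_Celebration_M montage.
# On customization load, an "unbrokentrash" face part sets the cowboy-hat flag and clears the ear-dynamics flag.
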AnimBlueprint /BRCosmetics/Characters/Player/Male/Medium/Heads/M_MED_FeralTrash_Head/Meshes/Head_FeralTrash_PostProcess_Export_Skeleton_AnimBlueprint
target_skeleton=/Game/Characters/Player/Male/Medium/Heads/M_MED_FeralTrash_Head/Meshes/Fortnite_Base_Head_FeralTrash_Export_Skeleton
graphs [
 AnimGraph {nodes [
   n1 = "Transform (Modify) Bone - Bone: head"(
    ComponentPose=Local To Component(LocalPose=Use cached pose 'inputPoseCache'())
    Alpha=Get enableHeadRotLimitAlpha() -> (enableHeadRotLimitAlpha:double)
    Translation="0,0,0"
    Rotation=Get headRotIsEmoting() -> (headRotIsEmoting:Rotator)
    Scale="1,1,1"
   ) {
    rotationMode=BMM_Additive
    rotationSpace=BCS_BoneSpace
    alphaBoolBlend {blendOption=Linear}
    boneToModify {boneName=head}
    rotation=<rot 50 50 50>
   }
   n2 = Component To Local(ComponentPose=@n1)
   n3 = "Transform (Modify) Bone - Bone: head"(
    ComponentPose=Local To Component(LocalPose=Use cached pose 'inputPoseCache'())
    Alpha=Get enableHeadRotLimitAlpha()
    Translation="0,0,0"
    Rotation=Get headRotIsNotEmoting() -> (headRotIsNotEmoting:Rotator)
    Scale="1,1,1"
   ) {rotationMode=BMM_Additive, rotationSpace=BCS_BoneSpace, alphaBoolBlend {blendOption=Linear}, boneToModify {boneName=head}}
   n4 = Component To Local(ComponentPose=@n3)
   n5 = Get bIsPlayingEmote()
   n6 = Anim Dynamics - Bone: dyn_ear_r_01!(
    ComponentPose=Local To Component(LocalPose=Blend Poses by bool(BlendPose_0=@n2, BlendPose_1=@n4, BlendTime_0=0.3, BlendTime_1=0.3, bActiveValue=@n5) {blendType=Linear})
    bAlphaBoolEnabled:bool=True
    ExternalForce="0,0,0"
   ) -> (Pose) {
    alphaInputType=Bool
    angularDampingOverride=1
    angularSpringConstant=10000
    bOverrideAngularDamping=true
    bUseGravityOverride=true
    gravityScale=0.35
    simulationSpace=RootRelative
    alphaBoolBlend {blendInTime=0.1, blendOption=Linear, blendOutTime=0.1}
    boundBone {boneName=dyn_ear_r_01}
    componentAppliedLinearAccClamp=<6000 6000 6000>
    componentLinearAccScale=<0.3 0.3 0.3>
    componentLinearVelScale=<0.3 0.3 0.3>
    gravityOverride {z=-980}
    physicsBodyDefinitions [{
      boundBone {boneName=dyn_ear_r_01}
      boxExtents=<15 20 15>
      constraintSetup {angularLimitsMax=<15 15 15>, angularLimitsMin=<-15 -15 -15>, angularTarget=<1 0 0>}
      localJointOffset=<2.5 0 0>
     }]
   }
   n7 = Anim Dynamics - Bone: dyn_cheek_r_01(ComponentPose=@n6, Alpha=1, ExternalForce="0,0,0") {
    angularDampingOverride=1
    angularSpringConstant=10000
    bAngularSpring=true
    bOverrideAngularDamping=true
    bUseGravityOverride=true
    gravityScale=0.35
    simulationSpace=RootRelative
    alphaBoolBlend {blendOption=Linear}
    boundBone {boneName=dyn_cheek_r_01}
    componentLinearAccScale=<1 1 1>
    componentLinearVelScale=<1 1 1>
    gravityOverride {z=-980}
    physicsBodyDefinitions [{
      boundBone {boneName=dyn_cheek_r_01}
      boxExtents=<20 30 30>
      constraintSetup {angularLimitsMax=<0 5 15>, angularLimitsMin=<0 -15 -15>, angularTarget=<1 0 0>}
      localJointOffset=<2.5 0 0>
     }]
   }
   n8 = Anim Dynamics - Bone: dyn_ear_l_01!(ComponentPose=@n7, bAlphaBoolEnabled:bool=True, ExternalForce="0,0,0") -> (Pose) {
    alphaInputType=Bool
    angularDampingOverride=1
    angularSpringConstant=10000
    bOverrideAngularDamping=true
    bUseGravityOverride=true
    gravityScale=0.35
    simulationSpace=RootRelative
    alphaBoolBlend {blendInTime=0.1, blendOption=Linear, blendOutTime=0.1}
    boundBone {boneName=dyn_ear_l_01}
    componentAppliedLinearAccClamp=<6000 6000 6000>
    componentLinearAccScale=<0.3 0.3 0.3>
    componentLinearVelScale=<0.3 0.3 0.3>
    gravityOverride {z=-980}
    physicsBodyDefinitions [{
      boundBone {boneName=dyn_ear_l_01}
      boxExtents=<15 20 15>
      constraintSetup {angularLimitsMax=<15 15 15>, angularLimitsMin=<-15 -15 -15>, angularTarget=<1 0 0>}
      localJointOffset=<2.5 0 0>
     }]
   }
   n9 = Anim Dynamics - Bone: dyn_cheek_l_01(ComponentPose=@n8, Alpha=1, ExternalForce="0,0,0") {
    angularDampingOverride=1
    angularSpringConstant=10000
    bAngularSpring=true
    bOverrideAngularDamping=true
    bUseGravityOverride=true
    gravityScale=0.35
    simulationSpace=RootRelative
    alphaBoolBlend {blendOption=Linear}
    boundBone {boneName=dyn_cheek_l_01}
    componentLinearAccScale=<1 1 1>
    componentLinearVelScale=<1 1 1>
    gravityOverride {z=-980}
    physicsBodyDefinitions [{
      boundBone {boneName=dyn_cheek_l_01}
      boxExtents=<20 30 30>
      constraintSetup {angularLimitsMax=<0 -5 25>, angularLimitsMin=<0 -20 5>, angularTarget=<1 0 0>}
      localJointOffset=<2.5 0 0>
     }]
   }
   n10 = Get isCustomEmotePlaying() -> (isCustomEmotePlaying:double)
   n11 = "Transform (Modify) Bone - Bone: L_lip_lower_outer"(ComponentPose=@n9, Alpha=@n10, Translation="0,0,0", Rotation="-40,0,0", Scale="1,1,1") {
    rotationMode=BMM_Additive
    rotationSpace=BCS_BoneSpace
    alphaBoolBlend {blendOption=Linear}
    alphaScaleBiasClamp {bInterpResult=true}
    boneToModify {boneName=L_lip_lower_outer}
    rotation {pitch=-40}
   }
   n12 = "Transform (Modify) Bone - Bone: R_lip_lower_outer"(ComponentPose=@n11, Alpha=@n10, Translation="0,0,0", Rotation="40,0,0", Scale="1,1,1") {
    rotationMode=BMM_Additive
    rotationSpace=BCS_BoneSpace
    alphaBoolBlend {blendOption=Linear}
    alphaScaleBiasClamp {bInterpResult=true}
    boneToModify {boneName=R_lip_lower_outer}
    rotation {pitch=40}
   }
   Output Pose(Result=Component To Local(ComponentPose=@n12)) {layerGroup=None}
   n13 = "Transform (Modify) Bone - Bone: dyn_ear_l_MSTR"!(
    ComponentPose=Local To Component(LocalPose=Input Pose())
    bAlphaBoolEnabled:bool=True
    Translation="11.3,-5,-8.25"
    Rotation="-50,30,100"
    Scale="1,1,1"
   ) -> (Pose) {
    alphaInputType=Bool
    rotationMode=BMM_Replace
    rotationSpace=BCS_ParentBoneSpace
    scaleMode=BMM_Additive
    translationMode=BMM_Replace
    translationSpace=BCS_ParentBoneSpace
    alphaBoolBlend {blendInTime=0.1, blendOption=Linear, blendOutTime=0.1}
    boneToModify {boneName=dyn_ear_l_MSTR}
    rotation=<rot -50 30 100>
    translation=<11.3 -5 -8.25>
   }
   n14 = "Transform (Modify) Bone - Bone: dyn_ear_r_MSTR"!(ComponentPose=@n13, bAlphaBoolEnabled:bool=True, Translation="11.3,-6,8.25", Rotation="60,40,105.5", Scale="1,1,1") -> (Pose) {
    alphaInputType=Bool
    rotationMode=BMM_Replace
    rotationSpace=BCS_ParentBoneSpace
    scaleMode=BMM_Additive
    translationMode=BMM_Replace
    translationSpace=BCS_ParentBoneSpace
    alphaBoolBlend {blendInTime=0.1, blendOption=Linear, blendOutTime=0.1}
    boneToModify {boneName=dyn_ear_r_MSTR}
    rotation=<rot 60 40 105.5>
    translation=<11.3 -6 8.25>
   }
   Save cached pose 'inputPoseCache'(Pose=Component To Local(ComponentPose=@n14)) {cacheName=inputPoseCache}
  ]}
 EventGraph {nodes [
   Event BlueprintUpdateAnimation()
   > n1 = Sequence!(execute) -> (then_0, then_1, then_2:Exec, then_3:Exec, then_4:Exec)
   > n2 = Sequence!(execute) -> (then_0, then_1, then_2, then_3, then_4, then_5:Exec, then_6:Exec, then_7:Exec)
   > n3 = Set headRotScaleIsEmoting(headRotScaleIsEmoting:Rotator="0, 0, 0") -> (Output_Get:Rotator)
   n4 = Set headRotScaleIsNotEmoting(execute=@n2.then_1, headRotScaleIsNotEmoting:Rotator="0.4,0.3,0.4") -> (Output_Get:Rotator)
   n5 = BreakRotator(InRot=@n3)
   Set X percentageEmote(execute=@n2.then_2, X percentageEmote:float=@n5) -> (Output_Get:float)
   Set Y percentageEmote(execute=@n2.then_3, Y percentageEmote:float=@n5.Pitch) -> (Output_Get:float)
   Set Z percentageEmote(execute=@n2.then_4, Z percentageEmote:float=@n5.Yaw) -> (Output_Get:float)
   n6 = BreakRotator(InRot=@n4)
   Set X percentageNotEmote(execute=@n2.then_5, X percentageNotEmote:float=@n6) -> (Output_Get:float)
   Set Y percentageNotEmote(execute=@n2.then_6, Y percentageNotEmote:float=@n6.Pitch) -> (Output_Get:float)
   Set Z percentageNotEmote(execute=@n2.then_7, Z percentageNotEmote:float=@n6.Yaw) -> (Output_Get:float)
   n7 = Is Valid(exec=@n1.then_1, InputObject=TryGetPawnOwner())
   Set enableHeadRotLimitAlpha(execute=@n7.Is Not Valid, enableHeadRotLimitAlpha:double=0) -> (Output_Get:double)
   Set enableHeadRotLimitAlpha(execute=@n7.Is Valid, enableHeadRotLimitAlpha=1)
   n8 = BreakTransform(InTransform=GetDeltaTransformFromRefPose(self=GetOwnerSkeletalMeshComponent(), BoneName=head, BaseName=neck_02))
   n9 = BreakRotator(InRot=@n8.Rotation)
   n10 = Get Z percentageEmote() -> (Z percentageEmote:float)
   n11 = Get X percentageEmote() -> (X percentageEmote:float)
   n12 = Get Y percentageEmote() -> (Y percentageEmote:float)
   > Set headRotIsEmoting(headRotIsEmoting:Rotator=MakeRotator(
     Roll=float * float!(A=@n9, B=-1, C:double=@n11) -> (ReturnValue)
     Pitch=float * float(A=@n9.Pitch, B=-1, C=@n12)
     Yaw=float * float(A=@n9.Yaw, B=-1, C=@n10)
    )) -> (Output_Get:Rotator)
   n13 = BreakRotator(InRot=@n8.Rotation)
   n14 = Get Y percentageNotEmote() -> (Y percentageNotEmote:float)
   n15 = Get X percentageNotEmote() -> (X percentageNotEmote:float)
   n16 = Get Z percentageNotEmote() -> (Z percentageNotEmote:float)
   > Set headRotIsNotEmoting(headRotIsNotEmoting:Rotator=MakeRotator(Roll=float * float(A=@n13, B=-1, C=@n15), Pitch=float * float(A=@n13.Pitch, B=-1, C=@n14), Yaw=float * float(A=@n13.Yaw, B=-1, C=@n16))) -> (Output_Get:Rotator)
   n17 = GetAnimInstance(self=GetOwnerSkeletalMeshComponent())
   Is Valid(exec=@n1.then_2, InputObject=@n17)
   > Set isCustomEmotePlaying(isCustomEmotePlaying:double=SelectFloat(A=1, B=0, bPickA=Montage_IsPlaying(self=@n17, Montage=/Game/Animation/Game/MainPlayer/Emotes/Emote_Celebration_M))) -> (Output_Get:double)
   n18 = Event OnCharacterCustomizationLoaded()
   n19 = Get Array Item with EFort Custom Part Type(Array:Array of Custom Character Part Object References=@n18.CustomParts, PartType=Face) -> (Output:Custom Character Part Object Reference)
   > n20 = Branch(Condition=MatchesWildcard(SourceString=GetDisplayName(Object=@n19), Wildcard=*unbrokentrash*))
   > Set Is Wearing Cowboy Hat(Is Wearing Cowboy Hat:bool=true) -> (Output_Get:bool)
   > Set Enable Ears Dynamics(Enable Ears Dynamics:bool=false) -> (Output_Get:bool)
   Set Is Wearing Cowboy Hat(execute=@n20.else, Is Wearing Cowboy Hat=false)
   > Set Enable Ears Dynamics(Enable Ears Dynamics=true)
  ]}
]
variables [
 BodyMesh {type=SkeletalMeshComponent}
 Enable Ears Dynamics {type=boolean}
 enableHeadRotLimitAlpha {type=number, default=1}
 head rot {type=Rotator}
 headRotIsEmoting {type=Rotator}
 headRotIsNotEmoting {type=Rotator}
 headRotScale {type=Rotator}
 headRotScaleIsEmoting {type=Rotator}
 headRotScaleIsNotEmoting {type=Rotator}
 Is Wearing Cowboy Hat {type=boolean}
 isCustomEmotePlaying {type=number}
 NewVar {type=AnimMontage, default=null}
 X percentage {type=number}
 X percentageEmote {type=number}
 X percentageNotEmote {type=number}
 Y percentage {type=number}
 Y percentageEmote {type=number}
 Y percentageNotEmote {type=number}
 Z percentage {type=number}
 Z percentageEmote {type=number}
 Z percentageNotEmote {type=number}
]
